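# Animation Blueprint for the Commando PumpkinSlice hat cosmetic, adding procedural secondary motion to its stem.
# It copies the pose from the attached parent mesh, then runs a half-alpha Anim Dynamics node on dyn_stem with linear and angular springs before output.
# A second Anim Dynamics node for dyn_stem1, with wind and gravity override, is left unconnected, as is an Input Pose node.
# The event graph has no logic.
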AnimBlueprint /BRCosmetics/Accessories/Hats/M_MED_Commando_PumpkinSlice_Hats/Meshes/M_MED_Commando_PumpkinSlice_Hats_Skeleton_AnimBlueprint
target_skeleton=/Game/Accessories/Hats/M_MED_Commando_PumpkinSlice_Hats/Meshes/M_MED_Commando_PumpkinSlice_Hats_Skeleton
graphs [
 AnimGraph {nodes [
   Anim Dynamics - Bone: dyn_stem1(Alpha=1, ExternalForce="0,0,0") {
    angularSpringConstant=1
    bEnableWind=true
    bGravityOverrideInSimSpace=true
    bUseGravityOverride=true
    alphaBoolBlend {blendOption=Linear}
    boundBone {boneName=dyn_stem1}
    componentAppliedLinearAccClamp=<1000 1000 1000>
    componentLinearAccScale=<0.25 0.25 0.25>
    componentLinearVelScale=<0.45 0.45 0.45>
    gravityOverride {z=-980}
    physicsBodyDefinitions [{
      sphereCollisionRadius=0
      boundBone {boneName=dyn_stem1}
      boxExtents=<10 10 10>
      constraintSetup {angularLimitsMax=<20 0 0>, angularLimitsMin=<-20 0 0>, angularTarget=<0 0 0>}
      localJointOffset=<0 0 -2>
     }]
   }
   n1 = Anim Dynamics - Bone: dyn_stem(ComponentPose=Local To Component(LocalPose=Copy Pose From Mesh() {bUseAttachedParent=true}), Alpha=0.5, ExternalForce="0,0,0") {
    angularSpringConstant=80
    bAngularSpring=true
    bGravityOverrideInSimSpace=true
    bLinearSpring=true
    lODThreshold=1
    linearSpringConstant=80
    alphaBoolBlend {blendOption=Linear}
    boundBone {boneName=dyn_stem}
    physicsBodyDefinitions [{
      sphereCollisionRadius=0
      boundBone {boneName=dyn_stem}
      boxExtents=<30 30 30>
      constraintSetup {angularTargetAxis=AxisY, twistAxis=AxisZ, angularLimitsMax=<30 30 0>, angularLimitsMin=<-30 -30 0>, angularTarget=<1 0 0>}
      localJointOffset=<0 0 -2>
     }]
   }
   Output Pose(Result=Component To Local(ComponentPose=@n1)) {layerGroup=None}
   AnimGraphNode_SubInput_1 = Input Pose()
  ]}
 EventGraph {nodes [Event BlueprintUpdateAnimation()]}
]
variables []
